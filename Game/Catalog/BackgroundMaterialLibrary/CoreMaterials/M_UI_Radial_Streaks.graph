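# Translucent UI background material that draws animated radial streaks over a radial color gradient, built for Item Shop tiles.
# A star field texture is panned in radial coordinates and tiled very unevenly, so its green channel reads as streaks.
# Shop background color parameters feed a smoothstepped gradient blend, and the streaks lerp that toward the gradient's second color.
# Shared functions supply rounded corners and vertex transform.
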Material /Game/Catalog/BackgroundMaterialLibrary/CoreMaterials/M_UI_Radial_Streaks
graph {
 properties {bAutomaticallySetUsageInEditor=false, blendMode=BLEND_Translucent, materialDomain=MD_UI, usedShadingModels=Default Lit}
 nodes [
  n1 = NamedRerouteUsage()
  n2 = MaterialFunctionCall(
   UV=MaterialFunctionCall() {materialFunction=&MF_UI_UVScaling}.Scaled To Fill
   GradientStart=Constant()
   GradientEnd=Constant() {r=1}
  ) {materialFunction=&MF_UI_RadialGradient}
  n3 = MaterialFunctionCall!(Vector or UVs) {materialFunction=/FortUILibrary/MaterialFunctions/Basics/MF_UI_CheaperVectorToRadial}
  n4 = AppendVector(A=ScalarParameter() {parameterName=U-Speed}, B=ScalarParameter() {defaultValue=0.5, parameterName=V-Speed})
  n5 = ScalarParameter() {defaultValue=0.4198, parameterName=G-U-Tiles}
  n6 = ScalarParameter() {defaultValue=0.0018, parameterName=G-V-Tiles}
  n7 = Multiply(
   A=Multiply(A=Panner(Coordinate=@n3.Radial Coordinates, Speed=@n4), B=AppendVector(A=@n5, B=@n6))
   B=ScalarParameter() {defaultValue=9, parameterName=TextureSize}
  )
  LinearInterpolate(
   A=LinearInterpolate(A=NamedRerouteUsage(), B=@n1, Alpha=SmoothStep(Value=@n2.Smooth) {constMax=1.2})
   B=@n1
   Alpha=Divide(A=TextureSample(UVs=@n7) {texture=/Game/Catalog/BackgroundMaterialLibrary/Textures/T_UI_StarField}.G)
  )
  MaterialFunctionCall!(Anchor, Translation, Padding, Scale) {materialFunction=/FortUILibrary/MaterialFunctions/Basics/MF_UI_VertexTransform}
  MaterialFunctionCall!(Input Alpha=Constant() {r=1}, bOverrideCornerRadius, Corner Radius Override) {materialFunction=/FortUILibrary/Systems/ItemShopTile/Materials/MaterialFunctions/MF_UI_ItemShop_RoundedCorners}
  NamedRerouteDeclaration(None=NamedRerouteUsage()) {name=Color2, nodeColor=<rgba 1 0.0007 0.2048 1>}
  n8 = MaterialFunctionCall() {materialFunction=/FortUILibrary/Systems/ItemShopTile/Materials/MaterialFunctions/MF_UI_ItemShop_BackgroundParams}
  NamedRerouteDeclaration(None=@n8.Color 2) {name=Secondary Color, nodeColor=<rgba 0 0 0 1>}
  NamedRerouteDeclaration(None=@n8.Color 3) {name=Accent Color, nodeColor=<rgba 0 0 0 1>}
  NamedRerouteDeclaration(None=NamedRerouteUsage()) {name=Color3, nodeColor=<rgba 0 1 0.625 1>}
  NamedRerouteDeclaration(None=@n8.Color 1) {name=Primary Color, nodeColor=<rgba 0 0 0 1>}
  NamedRerouteDeclaration(None=NamedRerouteUsage()) {name=Color1, nodeColor=<rgba 1 0.0007 0.2048 1>}
 ]
}
