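# Translucent UI material that draws an animated cartoon-cloud background for item shop tiles.
# Shared background-parameter colors feed vertical gradients, composited with texture layers through Overlay and Color Dodge blends.
# Clouds come from three particle-grid layers drifting at different speeds, each sampling a separate channel of one cloud texture.
# Supporting layers add a flowmap-distorted, panning watercolor texture, masked grunge and gradient-weighted random noise.
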Material /Game/Catalog/BackgroundMaterialLibrary/CoreMaterials/M_UI_Shop_Background_CartoonClouds
graph {
 properties {bAutomaticallySetUsageInEditor=false, blendMode=BLEND_Translucent, materialDomain=MD_UI, usedShadingModels=Default Lit}
 nodes [
  Constant3Vector() {constant=<rgba 0.4781 0.5999 0.7913 1>}
  Constant3Vector() {constant=<rgba 0.2329 0.2214 0.401 1>}
  Constant3Vector() {constant=<rgba 0.2604 0.4576 1 1>}
  Constant()
  MaterialFunctionCall!(Input Alpha=Constant() {r=1}, bOverrideCornerRadius, Corner Radius Override) {materialFunction=/FortUILibrary/Systems/ItemShopTile/Materials/MaterialFunctions/MF_UI_ItemShop_RoundedCorners}
  n1 = NamedRerouteUsage()
  n2 = TextureObject() {
   samplerType=SAMPLERTYPE_Alpha
   texture=/Game/Catalog/BackgroundMaterialLibrary/CoreMaterials/CrumpleFold/Textures/T_UI_DarkVortex_Background_Grunge
  }
  n3 = TextureSample(
   UVs=MaterialFunctionCall(UVs=@n1, Texture Scale=Constant() {r=0.3}) {materialFunction=&ScaleUVsByCenter}.UVs
   Tex=@n2
  ) {
   samplerType=SAMPLERTYPE_Alpha
   texture=&T_UI_DarkVortex_Background_Grunge
  }
  n4 = TextureSample(UVs=@n1, Tex=TextureObject() {samplerType=SAMPLERTYPE_Grayscale, texture=&LowResBlurredNoise}) {samplerType=SAMPLERTYPE_Grayscale}
  n5 = Multiply(A=@n4.R, B=@n4.R)
  n6 = Multiply(A=Multiply(A=@n3.R, B=LinearInterpolate(Alpha=@n5) {constA=0.1})) {constB=2}
  n7 = NamedRerouteUsage()
  NamedRerouteDeclaration(None=Saturate(None=LinearInterpolate(A=@n6, Alpha=@n7) {constB=0})) {name=Grunge Texture, nodeColor=<rgba 0 0 0 1>}
  n8 = Multiply(
   A=MaterialFunctionCall!(UV=NamedRerouteUsage(), Seed) {materialFunction=&MF_UI_Random}.Result
   B=ScalarParameter() {defaultValue=0.3, group=Noise, parameterName=Noise Opacity}
  ) {constB=0.65}
  n9 = ComponentMask(None=NamedRerouteUsage()) {g=true}
  n10 = LinearInterpolate(Alpha=Power(Base=OneMinus(None=@n9))) {constA=0.1}
  NamedRerouteDeclaration(None=Saturate(None=Multiply(A=@n8, B=@n10))) {name=Noise Overlay, nodeColor=<rgba 0 0 0 1>}
  n11 = NamedRerouteUsage()
  n12 = AppendVector(A=Add(A=ComponentMask(None=@n11) {r=true}) {constB=0}, B=Add(A=ComponentMask(None=@n11) {g=true}) {constB=0})
  n13 = ComponentMask(None=Time()) {r=true}
  n14 = MaterialFunctionCall!(
   UV=Multiply(A=@n12, B=Constant() {r=2})
   Global Offset=AppendVector(A=Multiply(A=@n13) {constB=0.01}, B=Constant())
   Global Rotation=Constant()
   Tiling=Constant() {r=2}
   Cell Offset=Constant() {r=0.2}
   Cell Scale=Constant() {r=0.3}
   Cell Rotation=Constant() {r=1}
   Random Seed
  ) {materialFunction=/FortUILibrary/MaterialFunctions/Basics/MF_UI_ParticleGrid}
  n15 = LinearInterpolate(A=@n14.Cells UV, Alpha=@n14.Cells UV) {constB=0}
  n16 = ComponentMask(None=Time()) {r=true}
  n17 = MaterialFunctionCall(
   UV=Multiply(A=@n12, B=Constant() {r=1.4})
   Global Offset=AppendVector(A=Multiply(A=@n16) {constB=0.0025}, B=Constant())
   Global Rotation=Constant() {r=2}
   Tiling=Constant() {r=2}
   Cell Offset=Constant()
   Cell Scale=Constant() {r=0.45}
   Cell Rotation=Constant()
   Random Seed=Constant() {r=1}
  ) {materialFunction=&MF_UI_ParticleGrid}
  n18 = LinearInterpolate(A=@n17.Cells UV, Alpha=@n17.Cells UV) {constB=0}
  n19 = Add(
   A=Multiply(A=TextureSample(UVs=@n15) {samplerType=SAMPLERTYPE_LinearColor, texture=/Game/Catalog/BackgroundMaterialLibrary/Textures/T_UI_CartoonCloud}.R) {constB=0.9}
   B=Multiply(A=TextureSample(UVs=@n18) {samplerType=SAMPLERTYPE_LinearColor, texture=&T_UI_CartoonCloud}.G) {constB=0.6}
  )
  n20 = ComponentMask(None=Time()) {r=true}
  n21 = MaterialFunctionCall(
   UV=Multiply(A=@n12, B=Constant() {r=1})
   Global Offset=AppendVector(A=Multiply(A=@n20) {constB=0.015}, B=Constant())
   Global Rotation=Constant()
   Tiling=Constant() {r=3}
   Cell Offset=Constant()
   Cell Scale=Constant() {r=0.4}
   Cell Rotation=Constant()
   Random Seed=Constant() {r=1}
  ) {materialFunction=&MF_UI_ParticleGrid}
  n22 = Multiply(A=TextureSample(UVs=LinearInterpolate(A=@n21.Cells UV, Alpha=@n21.Cells UV) {constB=0}) {samplerType=SAMPLERTYPE_LinearColor, texture=&T_UI_CartoonCloud}.B) {constB=0.75}
  NamedRerouteDeclaration(None=Saturate(None=Add(A=@n19, B=@n22))) {name=Cartoon Clouds, nodeColor=<rgba 0 0 0 1>}
  n23 = NamedRerouteUsage()
  n24 = Multiply(A=@n23, B=Constant() {r=2}) {constB=0.65}
  n25 = Panner(Coordinate=@n24) {speedX=0.01, speedY=0.01}
  n26 = TextureObjectParameter(UVs=@n23) {
   group=04 - Foreground Clouds
   parameterName=Foreground Cloud Flowmap
   samplerType=SAMPLERTYPE_Alpha
   texture=/Game/Catalog/BackgroundMaterialLibrary/Textures/T_UI_Watercolor
  }
  n27 = Time()
  n28 = Constant() {r=0.001}
  n29 = Constant() {r=1}
  n30 = TextureObjectParameter() {
   group=04 - Foreground Clouds
   parameterName=Foreground Cloud Flowmap
   samplerType=SAMPLERTYPE_Alpha
   texture=&T_UI_Watercolor
  }
  n31 = TextureSample(UVs=Panner(Coordinate=@n24, Time=Multiply(A=@n27) {constB=0.001}, Speed=AppendVector(A=@n28, B=@n29)) {speedX=0.01, speedY=0.01}, Tex=@n30)
  NamedRerouteDeclaration(None=Saturate(None=MaterialFunctionCall!(UVs=@n25, Texture Object=@n26, Grayscale Flow Map=@n31.RGB, Flow Direction, Flow Intensity, Flow Speed) {materialFunction=/FortUILibrary/MaterialFunctions/Effects/MF_UI_GrayscaleFlowMap_New}.Distorted Texture)) {name=Watercolor Texture, nodeColor=<rgba 0 0 0 1>}
  n32 = MaterialFunctionCall() {materialFunction=/FortUILibrary/Systems/ItemShopTile/Materials/MaterialFunctions/MF_UI_ItemShop_BackgroundParams}
  NamedRerouteDeclaration(None=@n32.Color 1) {name=Primary Color, nodeColor=<rgba 0 0 0 1>}
  NamedRerouteDeclaration(None=@n32.Color 2) {name=Secondary Color, nodeColor=<rgba 0 0 0 1>}
  n33 = MaterialFunctionCall(UV=NamedRerouteDeclaration(None=MaterialFunctionCall() {materialFunction=&GetUserInterfaceUV}.Normalized UV) {name=Normalized UV, nodeColor=<rgba 0 0 0 1>}) {materialFunction=&MF_UI_UVScaling}
  NamedRerouteDeclaration(None=@n33.ScaleY) {name=Scale Y UV, nodeColor=<rgba 0 0 0 1>}
  NamedRerouteDeclaration(None=@n32.Color 3) {name=Accent Color, nodeColor=<rgba 0 0 0 1>}
  NamedRerouteDeclaration(None=@n33.Scaled To Fill) {name=Scale to Fill, nodeColor=<rgba 0 0 0 1>}
  NamedRerouteDeclaration(None=@n33.Scaled to Fit) {name=Scale to Fit, nodeColor=<rgba 0 0 0 1>}
  n34 = MaterialFunctionCall!(ScreenUV, TextureSize, ScalePivotPoint) {materialFunction=/FortUILibrary/MaterialFunctions/Basics/MF_UI_ScreenSpaceUVScaling}
  NamedRerouteDeclaration(None=@n34.ScaleY) {name="Scale Y UV  Screen", nodeColor=<rgba 0 0 0 1>}
  NamedRerouteDeclaration(None=@n34.Scaled to Fill) {name="Scale to Fill  Screen", nodeColor=<rgba 0 0 0 1>}
  NamedRerouteDeclaration(None=@n34.Scaled to Fit) {name=Scale to Fit Screen, nodeColor=<rgba 0 0 0 1>}
  NamedRerouteDeclaration(None=@n33.ScaleX) {name=Scale X UV, nodeColor=<rgba 0 0 0 1>}
  n35 = NamedRerouteUsage()
  n36 = MaterialFunctionCall(Input=ComponentMask(None=@n35) {g=true}, Input Low=Constant(), Input High=Constant() {r=1}) {materialFunction=&RemapValue_0-1}
  NamedRerouteDeclaration(None=LinearInterpolate(A=NamedRerouteUsage(), B=NamedRerouteUsage(), Alpha=Saturate(None=@n36.Result))) {name=Base Gradient, nodeColor=<rgba 0 0 0 1>}
  n37 = Saturate(None=MaterialFunctionCall(Input=ComponentMask(None=@n35) {g=true}, Input Low=Constant(), Input High=Constant() {r=1}) {materialFunction=&RemapValue_0-1}.Result)
  NamedRerouteDeclaration(None=LinearInterpolate(A=NamedRerouteUsage(), B=NamedRerouteUsage(), Alpha=@n37)) {name=Overlay Gradient, nodeColor=<rgba 0 0 0 1>}
  NamedRerouteDeclaration(None=@n37) {name=Gradient Mask, nodeColor=<rgba 0 0 0 1>}
  NamedRerouteUsage()
  NamedRerouteUsage()
  n38 = NamedRerouteUsage()
  n39 = Multiply(A=NamedRerouteUsage()) {constB=0.4}
  n40 = Add(
   A=MaterialFunctionCall(Base=NamedRerouteUsage(), Blend=LinearInterpolate(A=@n38, Alpha=@n39) {constA=1}) {materialFunction=&Blend_Overlay}.Result
   B=Multiply(A=NamedRerouteUsage()) {constB=0.2}
  )
  n41 = NamedRerouteUsage()
  n42 = NamedRerouteUsage()
  Saturate(None=MaterialFunctionCall!(Blend=NamedRerouteUsage(), Base=LinearInterpolate(A=@n40, B=@n41, Alpha=@n42)) {materialFunction=/Engine/Functions/Engine_MaterialFunctions03/Blends/Blend_ColorDodge}.Result)
 ]
}
