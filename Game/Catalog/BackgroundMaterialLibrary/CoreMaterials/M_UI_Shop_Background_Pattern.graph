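# Translucent UI material for item shop backgrounds, layering an animated tiled pattern and moving sheen over a colored gradient.
# A linear or radial gradient blends colors from the shop's shared BackgroundParams function, which also supplies an accent color.
# A particle grid function scrolls and tiles an alpha texture (clover by default), with per-cell random scale, rotation and optionally fluctuating opacity.
# Static switches select gradient type and screen-space pattern UVs; the file also includes a rounded-corners function.
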Material /Game/Catalog/BackgroundMaterialLibrary/CoreMaterials/M_UI_Shop_Background_Pattern
graph {
 properties {bAutomaticallySetUsageInEditor=false, blendMode=BLEND_Translucent, materialDomain=MD_UI, usedShadingModels=Default Lit}
 nodes [
  n1 = NamedRerouteUsage()
  n2 = NamedRerouteUsage()
  LinearInterpolate(
   A=@n1
   B=Saturate(None=MaterialFunctionCall!(Base=@n1, Add=@n2) {materialFunction=/Engine/Functions/Engine_MaterialFunctions02/Math/MultiplyAdd}.Result)
   Alpha=LinearInterpolate(B=NamedRerouteUsage(), Alpha=ScalarParameter() {defaultValue=1, group=04 - Sheen, parameterName=Apply Sheen on Pattern, sliderMax=1}) {constA=1}
  )
  MaterialFunctionCall!(Input Alpha=Constant() {r=1}, bOverrideCornerRadius, Corner Radius Override) {materialFunction=/FortUILibrary/Systems/ItemShopTile/Materials/MaterialFunctions/MF_UI_ItemShop_RoundedCorners}
  n3 = MaterialFunctionCall() {materialFunction=/FortUILibrary/Systems/ItemShopTile/Materials/MaterialFunctions/MF_UI_ItemShop_BackgroundParams}
  NamedRerouteDeclaration(None=@n3.Color 3) {name=Accent Color, nodeColor=<rgba 0 0 0 1>}
  NamedRerouteDeclaration(None=@n3.Color 1) {name=Primary Color, nodeColor=<rgba 0 0 0 1>}
  n4 = StaticSwitchParameter(
   True=MaterialFunctionCall!(ScreenUV, TextureSize, ScalePivotPoint) {materialFunction=/FortUILibrary/MaterialFunctions/Basics/MF_UI_ScreenSpaceUVScaling}.Scaled to Fit
   False=NamedRerouteUsage()
  ) {group=02 - Pattern, parameterName=Screen Space Pattern}
  n5 = Add(
   A=MaterialFunctionCall(UVs=NamedRerouteUsage(), Texture Scale=ScalarParameter() {defaultValue=1, group=01 - Gradient, parameterName=Radial Gradient Scale}) {materialFunction=&ScaleUVsByCenter}.UVs
   B=AppendVector(
    A=ScalarParameter() {group=01 - Gradient, parameterName=Radial Gradient Offset X}
    B=ScalarParameter() {group=01 - Gradient, parameterName=Radial Gradient Offset Y}
   )
  ) {constB=0.5}
  n6 = StaticSwitchParameter(
   True=MaterialFunctionCall(UVs=@n4, RotationAngle=ScalarParameter() {group=01 - Gradient, parameterName=Gradient Rotation}) {materialFunction=&MF_UI_LinearGradient}.Result
   False=MaterialFunctionCall(UV=MaterialFunctionCall(UV=@n5) {materialFunction=&MF_UI_UVScaling}.Scaled To Fill) {materialFunction=&MF_UI_RadialGradient}.Smooth
  ) {group=01 - Gradient, parameterName=Use Linear Gradient, sortPriority=-32}
  NamedRerouteDeclaration(None=LinearInterpolate(A=NamedRerouteUsage(), B=NamedRerouteUsage(), Alpha=Saturate(None=@n6))) {name=Colored Gradient, nodeColor=<rgba 0 0 0 1>}
  n7 = NamedRerouteUsage()
  n8 = NamedRerouteUsage()
  n9 = Frac(None=Multiply(A=ScalarParameter() {defaultValue=0.015, group=02 - Pattern, parameterName=Pattern Offset Speed}, B=@n8) {constB=0.015})
  n10 = NamedRerouteUsage()
  n11 = ScalarParameter() {defaultValue=0.015, group=03 - Cell, parameterName=Cell Rotation Speed}
  n12 = Add(A=ScalarParameter() {group=03 - Cell, parameterName=Cell Default Rotation, sliderMax=1}, B=Frac(None=Multiply(A=@n8, B=@n11)))
  n13 = MaterialFunctionCall!(
   UV=Multiply(A=@n4, B=ScalarParameter() {defaultValue=0.5, group=02 - Pattern, parameterName=Pattern Scale})
   Global Offset=AppendVector(A=Add(A=@n7, B=@n9), B=Constant())
   Global Rotation=ScalarParameter() {defaultValue=0.35, group=02 - Pattern, parameterName=Pattern Rotation}
   Tiling=Floor(None=ScalarParameter() {defaultValue=5, group=02 - Pattern, parameterName=Pattern Tiling})
   Cell Offset=Constant()
   Cell Scale=LinearInterpolate(
    A=ScalarParameter() {defaultValue=0.5, group=03 - Cell, parameterName=Minimum Cell Scale}
    B=ScalarParameter() {defaultValue=1, group=03 - Cell, parameterName=Maximum Cell Scale}
    Alpha=@n10
   ) {constA=0.5}
   Cell Rotation=LinearInterpolate(A=@n12, B=Add(A=@n10, B=@n12), Alpha=ScalarParameter() {defaultValue=1, group=0 - Global, parameterName=Randomize Rotation, sliderMax=1})
   Random Seed
  ) {materialFunction=/FortUILibrary/MaterialFunctions/Basics/MF_UI_ParticleGrid}
  n14 = TextureObjectParameter() {
   group=02 - Pattern
   parameterName=Pattern Texture
   samplerType=SAMPLERTYPE_Alpha
   texture=/Game/UI/Frontend/ItemShop/Art/Lucky/T_UI_LuckyClover-01
  }
  n15 = TextureSample(UVs=NamedRerouteDeclaration(None=@n13.Cells UV) {name=Cells UV, nodeColor=<rgba 0 0 0 1>}, Tex=@n14)
  n16 = ScalarParameter() {defaultValue=0.15, group=03 - Cell, parameterName=Cell Min Opacity, sliderMax=1}
  n17 = ScalarParameter() {defaultValue=1, group=03 - Cell, parameterName=Cell Max Opacity, sliderMax=1}
  n18 = NamedRerouteUsage()
  n19 = LinearInterpolate(
   B=NamedRerouteUsage()
   Alpha=Multiply(A=NamedRerouteUsage(), B=ScalarParameter() {defaultValue=1, group=03 - Cell, parameterName=Cell Opacity Fluctuation Speed}) {constB=0.15}
  ) {constAlpha=1}
  n20 = LinearInterpolate(
   A=Max(A=Round(None=@n18), B=@n18)
   B=Saturate(None=MaterialFunctionCall(Speed=@n19) {materialFunction=&DebugTimeSine}.Result)
   Alpha=ScalarParameter() {group=03 - Cell, parameterName=Fluctuate Cell Opacity, sliderMax=1}
  )
  NamedRerouteDeclaration(None=Multiply(
    A=NamedRerouteDeclaration(None=@n15.R) {name=Pattern Mask, nodeColor=<rgba 0 0 0 1>}
    B=LinearInterpolate(A=@n16, B=@n17, Alpha=@n20) {constA=0.15}
   )) {name=Pattern Opacity, nodeColor=<rgba 0 0 0 1>}
  n21 = NamedRerouteUsage()
  n22 = ScalarParameter() {group=04 - Sheen, parameterName=Sheen Angle}
  n23 = LinearInterpolate(Alpha=MaterialFunctionCall!(
    Animation Duration=ScalarParameter() {defaultValue=1, group=04 - Sheen, parameterName=Sheen Duration}
    Delay=ScalarParameter() {defaultValue=3, group=04 - Sheen, parameterName=Sheen Delay}
    Reset to Default
   ) {materialFunction=/FortUILibrary/MaterialFunctions/Time/MF_UI_AnimationDelay}.Result) {constA=-1}
  n24 = Constant()
  n25 = MaterialFunctionCall(
   UVs=Add(A=Rotator(Coordinate=@n21, Time=@n22) {speed=6.2832}, B=AppendVector(A=@n23, B=@n24))
   Texture Scale=AppendVector(A=ScalarParameter() {defaultValue=0.2, group=04 - Sheen, parameterName=Sheen Scale}, B=Constant() {r=1})
  ) {materialFunction=&ScaleUVsByCenter}
  n26 = ComponentMask(None=MaterialFunctionCall!(Value=@n25.UVs, Scale=Constant() {r=1}, Height=Constant() {r=1}) {materialFunction=/FortUILibrary/MaterialFunctions/Basics/MF_UI_PingPong}.Result) {r=true}
  NamedRerouteDeclaration(None=Multiply(
    A=MaterialFunctionCall!(UVs=@n25.UVs, Alpha=@n26) {materialFunction=/FortUILibrary/MaterialFunctions/Basics/MF_UI_ClipEdgePixel}.Result
    B=ScalarParameter() {defaultValue=1, group=04 - Sheen, parameterName=Sheen Opacity, sliderMax=1}
   )) {name=Sheen Opacity, nodeColor=<rgba 0 0 0 1>}
  n27 = NamedRerouteUsage()
  n28 = NamedRerouteUsage()
  n29 = ScalarParameter() {defaultValue=1, group=02 - Pattern, parameterName=Pattern Blend, sliderMax=1, sliderMin=-1}
  n30 = NamedRerouteUsage()
  n31 = Saturate(None=MaterialFunctionCall(Base=@n27, Add=Multiply(A=@n29, B=@n30) {constB=2}) {materialFunction=&MultiplyAdd}.Result)
  n32 = ScalarParameter() {group=02 - Pattern, parameterName=Blend Pattern and Gradient, sliderMax=1}
  NamedRerouteDeclaration(None=LinearInterpolate(A=@n27, B=LinearInterpolate(A=@n28, B=@n31, Alpha=@n32), Alpha=NamedRerouteUsage())) {name=Pattern Color, nodeColor=<rgba 0 0 0 1>}
  n33 = ScalarParameter() {defaultValue=1, group=0 - Global, parameterName=Global Speed}
  NamedRerouteDeclaration(None=@n33) {name=Global Speed Input, nodeColor=<rgba 0 0 0 1>}
  n34 = MaterialFunctionCall(UV=NamedRerouteDeclaration(None=MaterialFunctionCall() {materialFunction=&GetUserInterfaceUV}.Normalized UV) {name=Normalized UV, nodeColor=<rgba 0 0 0 1>}) {materialFunction=&MF_UI_UVScaling}
  NamedRerouteDeclaration(None=@n34.Scaled To Fill) {name=Scale to Fill UV, nodeColor=<rgba 0 0 0 1>}
  NamedRerouteDeclaration(None=@n3.Color 2) {name=Secondary Color, nodeColor=<rgba 0 0 0 1>}
  NamedRerouteDeclaration(None=@n34.ScaleY) {name=Scale Y UV, nodeColor=<rgba 0 0 0 1>}
  NamedRerouteDeclaration(None=@n13.Random Rows) {name=Random Rows, nodeColor=<rgba 0 0 0 1>}
  NamedRerouteDeclaration(None=Multiply(A=Time(), B=@n33)) {name=Global Speed, nodeColor=<rgba 0 0 0 1>}
  NamedRerouteDeclaration(None=@n13.Random Cell Values) {name=Random Cell Values, nodeColor=<rgba 0 0 0 1>}
  NamedRerouteUsage()
 ]
}
