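# Translucent UI material that draws an animated storm-cloud background, apparently for item shop tiles, using shared background color parameters and a rounded-corner function.
# Two noise samples panning in opposite directions build a flow map, with distortion parameters for two scrolling samples of one cloud texture.
# The contrast-boosted layers multiply into a blend mask, with over-bright areas subtracted back.
# A rotated gradient, color lerps and a positionable radial spotlight provide the shading.
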Material /Game/Catalog/BackgroundMaterialLibrary/CoreMaterials/M_UI_Shop_Background_Storm
graph {
 properties {bAutomaticallySetUsageInEditor=false, blendMode=BLEND_Translucent, materialDomain=MD_UI, usedShadingModels=Default Lit}
 nodes [
  n1 = Multiply(A=NamedRerouteUsage(), B=ScalarParameter() {defaultValue=0.7, group=FlowMap, parameterName=FlowMapTiling})
  n2 = AppendVector(A=ComponentMask(None=@n1) {g=true}, B=ComponentMask(None=@n1) {r=true})
  n3 = ScalarParameter() {defaultValue=0.03, group=FlowMap, parameterName=FlowMapSpeedX}
  n4 = ScalarParameter() {defaultValue=0.05, group=FlowMap, parameterName=FlowMapSpeedY}
  n5 = Frac(None=Multiply(A=Time(), B=AppendVector(A=@n3, B=@n4)) {constB=-0.07})
  n6 = TextureObjectParameter() {
   group=FlowMap
   parameterName=FlowMapTexture
   samplerType=SAMPLERTYPE_Grayscale
   texture=&LowResBlurredNoise
  }
  n7 = Add(A=TextureSample(UVs=Subtract(A=@n2, B=@n5), Tex=@n6).R) {constB=-0.5}
  n8 = Add(A=TextureSample(UVs=Add(A=@n5, B=@n1), Tex=@n6).R) {constB=-0.5}
  Add(A=@n7, B=@n8)
  n9 = NamedRerouteUsage()
  n10 = LinearInterpolate(A=NamedRerouteUsage(), B=NamedRerouteUsage(), Alpha=@n9)
  n11 = NamedRerouteUsage()
  n12 = ScalarParameter() {defaultValue=0.1, group=General, parameterName=BaseBrightness}
  n13 = Add(
   A=MaterialFunctionCall(UVs=NamedRerouteUsage(), Texture Scale=ScalarParameter() {defaultValue=1, group=Spotlight, parameterName=SpotlightScale}) {materialFunction=&ScaleUVsByCenter}.UVs
   B=AppendVector(A=ScalarParameter() {group=Spotlight, parameterName=SpotlightX}, B=ScalarParameter() {group=Spotlight, parameterName=SpotlightY})
  )
  LinearInterpolate(
   A=Add(A=Multiply(A=@n10, B=@n9), B=Multiply(A=@n11, B=@n12) {constB=0.1})
   B=NamedRerouteUsage()
   Alpha=Multiply(
    A=MaterialFunctionCall(UV=@n13) {materialFunction=&MF_UI_RadialGradient}.Smooth
    B=ScalarParameter() {defaultValue=1, group=Spotlight, parameterName=SpotlightOpacity}
   )
  )
  MaterialFunctionCall!(Input Alpha=Constant() {r=1}, bOverrideCornerRadius, Corner Radius Override) {materialFunction=/FortUILibrary/Systems/ItemShopTile/Materials/MaterialFunctions/MF_UI_ItemShop_RoundedCorners}
  n14 = MaterialFunctionCall(UVs=NamedRerouteUsage(), "Rotation Angle (0-1)"=ScalarParameter() {defaultValue=0.12, group=Gradient, parameterName=GradientAngle}) {materialFunction=&CustomRotator}
  n15 = Saturate(None=Add(A=ComponentMask(None=@n14.Rotated Values) {g=true}, B=ScalarParameter() {defaultValue=-0.3, group=Gradient, parameterName=GradientOffset}) {constB=-0.3})
  Multiply(A=Saturate(None=Multiply(A=@n15, B=@n15)), B=ScalarParameter() {defaultValue=1, group=Gradient, parameterName=BlendTexturesAndGradient})
  n16 = MaterialFunctionCall() {materialFunction=/FortUILibrary/Systems/ItemShopTile/Materials/MaterialFunctions/MF_UI_ItemShop_BackgroundParams}
  NamedRerouteDeclaration(None=@n16.Color 3) {name=Accent Color, nodeColor=<rgba 0 0 0 1>}
  NamedRerouteDeclaration(None=@n16.Color 1) {name=Primary Color, nodeColor=<rgba 0 0 0 1>}
  n17 = NamedRerouteUsage()
  n18 = Length(None=@n17) {is=class MaterialExpressionLength}
  NamedRerouteDeclaration(None=Multiply(A=Multiply(A=@n18) {constB=-1}, B=Multiply(A=@n7, B=@n8))) {name=FlowMap, nodeColor=<rgba 0 0.2734 1 1>}
  NamedRerouteDeclaration(None=@n16.Color 2) {name=Secondary Color, nodeColor=<rgba 0 0 0 1>}
  n19 = MaterialFunctionCall!(ScreenUV, TextureSize, ScalePivotPoint) {materialFunction=/FortUILibrary/MaterialFunctions/Basics/MF_UI_ScreenSpaceUVScaling}
  NamedRerouteDeclaration(None=@n19.ScaleX) {name=Scale X UV, nodeColor=<rgba 0 0 0 1>}
  n20 = MaterialFunctionCall() {materialFunction=&GetUserInterfaceUV}
  NamedRerouteDeclaration(None=@n20.9-Slice UV) {name=9 Slice UV, nodeColor=<rgba 0 0 0 1>}
  NamedRerouteDeclaration(None=@n19.ScaleY) {name=Scale Y UV, nodeColor=<rgba 0 0 0 1>}
  n21 = Multiply(
   A=AppendVector(
    A=ScalarParameter() {defaultValue=-0.01, group=Texture 1, parameterName=Tex1SpeedX}
    B=ScalarParameter() {defaultValue=0.005, group=Texture 1, parameterName=Tex1SpeedY}
   )
   B=Time()
  )
  n22 = NamedRerouteUsage()
  n23 = AppendVector(
   A=ScalarParameter() {defaultValue=0.8, group=Texture 1, parameterName=Tex1TilingX}
   B=ScalarParameter() {defaultValue=0.8, group=Texture 1, parameterName=Tex1TilingY}
  )
  n24 = Add(
   A=Multiply(A=NamedRerouteUsage(), B=ScalarParameter() {defaultValue=0.07, group=Texture 1, parameterName=Tex1Distortion}) {constB=0.015}
   B=Add(A=Frac(None=@n21), B=Multiply(A=@n22, B=@n23) {constB=3})
  )
  n25 = TextureObjectParameter() {group=General, parameterName=VortexTexture, texture=/Game/UI/Frontend/ItemShop/Art/A_Shop_Tiles_Textures/T_UI_StormClouds}
  n26 = ScalarParameter() {defaultValue=2, group=Texture 1, parameterName=Tex1BrightnessVariation}
  n27 = NamedRerouteUsage()
  n28 = Add(
   A=Add(A=TextureSample(UVs=@n24, Tex=@n25) {samplerType=SAMPLERTYPE_LinearColor}.R, B=Multiply(A=@n26, B=@n27) {constB=0.6}) {constB=0}
   B=ScalarParameter() {defaultValue=0.1, group=Texture 1, parameterName=Tex1BaseBrightness}
  ) {constB=0.1}
  n29 = ScalarParameter() {defaultValue=1.5, group=Texture 1, parameterName=Tex1Contrast}
  n30 = NamedRerouteUsage()
  n31 = ScalarParameter() {defaultValue=2, group=Texture 2, parameterName=Tex2BrightnessVariation}
  n32 = Multiply(
   A=AppendVector(
    A=ScalarParameter() {defaultValue=-0.01, group=Texture 2, parameterName=Tex2SpeedRadial}
    B=ScalarParameter() {defaultValue=0.01, group=Texture 2, parameterName=Tex2SpeedAngular}
   )
   B=Time()
  )
  n33 = NamedRerouteUsage()
  n34 = AppendVector(
   A=ScalarParameter() {defaultValue=0.7, group=Texture 2, parameterName=Tex2TilingX}
   B=ScalarParameter() {defaultValue=0.7, group=Texture 2, parameterName=Tex2TilingY}
  )
  n35 = Add(
   A=Multiply(A=NamedRerouteUsage(), B=ScalarParameter() {defaultValue=-0.1, group=Texture 2, parameterName=Tex2Distortion}) {constB=-0.02}
   B=Add(A=Frac(None=@n32), B=Multiply(A=@n33, B=@n34) {constB=3})
  )
  n36 = Add(
   A=Add(A=Multiply(A=@n30, B=@n31) {constB=-0.7}, B=TextureSample(UVs=@n35, Tex=@n25) {samplerType=SAMPLERTYPE_LinearColor}.R) {constB=-0.2}
   B=ScalarParameter() {defaultValue=0.1, group=Texture 2, parameterName=Tex2BaseBrightness}
  ) {constB=0.1}
  n37 = ScalarParameter() {defaultValue=2, group=Texture 2, parameterName=Tex2Contrast}
  n38 = Abs(None=Multiply(A=Multiply(A=@n28, B=@n29) {constB=1.5}, B=Multiply(A=@n36, B=@n37) {constB=2}))
  n39 = Saturate(None=Multiply(A=@n38, B=@n38) {constB=-1})
  n40 = Multiply(
   A=Max(A=Add(A=@n38) {constB=-1.05}) {constB=0}
   B=ScalarParameter() {defaultValue=0.7, group=General, parameterName=SaturationInversionStrength}
  ) {constB=0.7}
  NamedRerouteDeclaration(None=Saturate(None=Subtract(A=@n39, B=@n40))) {name=TexturesBlend, nodeColor=<rgba 0 0 0 1>}
  NamedRerouteDeclaration(None=MaterialFunctionCall(UV=NamedRerouteDeclaration(None=@n20.Normalized UV) {name=Normalized UV, nodeColor=<rgba 0 0 0 1>}) {materialFunction=&MF_UI_UVScaling}.Scaled to Fit) {name=Fit UV Y, nodeColor=<rgba 0 0 0 1>}
  NamedRerouteDeclaration(None=@n19.Scaled to Fill) {name=Scale to Fill, nodeColor=<rgba 0 0 0 1>}
  NamedRerouteUsage()
  NamedRerouteUsage()
  NamedRerouteUsage()
  NamedRerouteUsage()
  Subtract(A=@n17, B=NamedRerouteUsage())
 ]
}
