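# Translucent UI material for item shop backgrounds, layering animated streaks over a gradient with an optional background texture.
# Shared shop functions supply the color palette and rounded corners, and a static switch picks a linear or radial gradient.
# Streaks come from a rotated, scaled particle grid scrolled by time, sampling a channel-masked texture sharpened with a power curve.
# Intensity, blend and opacity parameters control how streaks composite over the background.
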Material /Game/Catalog/BackgroundMaterialLibrary/CoreMaterials/M_UI_Shop_Background_Streaks
graph {
 properties {bAutomaticallySetUsageInEditor=false, blendMode=BLEND_Translucent, materialDomain=MD_UI, usedShadingModels=Default Lit}
 nodes [
  n1 = NamedRerouteUsage()
  n2 = Multiply(A=ScalarParameter() {defaultValue=1, group=03 - Streaks, parameterName=Streaks Blend, sliderMax=1, sliderMin=-1}, B=NamedRerouteUsage()) {constB=2}
  n3 = LinearInterpolate(
   A=NamedRerouteUsage()
   B=Saturate(None=MaterialFunctionCall!(Base=@n1, Add=@n2) {materialFunction=/Engine/Functions/Engine_MaterialFunctions02/Math/MultiplyAdd}.Result)
   Alpha=ScalarParameter() {group=03 - Streaks, parameterName=Blend Streaks and Gradient, sliderMax=1}
  )
  n4 = LinearInterpolate(B=ScalarParameter() {defaultValue=40, group=03 - Streaks, parameterName=Streaks Color Intensity}, Alpha=NamedRerouteUsage()) {constA=1}
  LinearInterpolate(
   A=NamedRerouteUsage()
   B=Saturate(None=Multiply(A=@n3, B=@n4) {constB=40})
   Alpha=Multiply(A=NamedRerouteUsage(), B=ScalarParameter() {defaultValue=1, group=03 - Streaks, parameterName=Streaks Opacity, sliderMax=1})
  )
  MaterialFunctionCall!(Input Alpha=Constant() {r=1}, bOverrideCornerRadius, Corner Radius Override) {materialFunction=/FortUILibrary/Systems/ItemShopTile/Materials/MaterialFunctions/MF_UI_ItemShop_RoundedCorners}
  n5 = ScalarParameter() {defaultValue=1, group=02 - Background Texture, parameterName=Background Texture Blend, sliderMax=1, sliderMin=-1}
  n6 = TextureObjectParameter() {
   group=02 - Background Texture
   parameterName=Background Texture
   samplerType=SAMPLERTYPE_Alpha
   texture=/Game/Catalog/BackgroundMaterialLibrary/Textures/T_UI_Shop_Streaks_Background_Gimmick
  }
  n7 = TextureSample(UVs=NamedRerouteUsage(), Tex=@n6) {samplerSource=SSM_Clamp_WorldGroupSettings}
  n8 = Saturate(None=MaterialFunctionCall(Base=NamedRerouteUsage(), Add=Multiply(A=@n5, B=@n7.R)) {materialFunction=&MultiplyAdd}.Result)
  n9 = Multiply(A=ScalarParameter() {defaultValue=0.1, group=02 - Background Texture, parameterName=Background Texture Min, sliderMax=1}, B=@n8) {constB=0.1}
  n10 = Power(Base=@n7.R, Exp=ScalarParameter() {defaultValue=1, group=02 - Background Texture, parameterName=Background Texture Contrast}) {constExponent=1}
  NamedRerouteDeclaration(None=StaticSwitchParameter(True=LinearInterpolate(A=@n9, B=@n8, Alpha=@n10), False=NamedRerouteUsage()) {group=02 - Background Texture, parameterName=Use Background Texture, sortPriority=-32}) {name=Colored Background Texture, nodeColor=<rgba 0 0 0 1>}
  n11 = MaterialFunctionCall() {materialFunction=/FortUILibrary/Systems/ItemShopTile/Materials/MaterialFunctions/MF_UI_ItemShop_BackgroundParams}
  NamedRerouteDeclaration(None=@n11.Color 1) {name=Primary Color, nodeColor=<rgba 0 0 0 1>}
  n12 = NamedRerouteUsage()
  n13 = ScalarParameter() {defaultValue=0.125, group=03 - Streaks, parameterName=Streaks Angle}
  n14 = ScalarParameter() {defaultValue=0.7, group=03 - Streaks, parameterName=Streaks Scale X}
  n15 = ScalarParameter() {defaultValue=3, group=03 - Streaks, parameterName=Streaks Scale Y}
  n16 = Time()
  n17 = Multiply(A=ScalarParameter() {defaultValue=0.4, group=03 - Streaks, parameterName=Streaks Offset Speed}) {constB=-1}
  n18 = MaterialFunctionCall!(
   UV=MaterialFunctionCall(UVs=Rotator(Coordinate=@n12, Time=@n13) {speed=6.2832}, Texture Scale=AppendVector(A=@n14, B=@n15)) {materialFunction=&ScaleUVsByCenter}.UVs
   Global Offset=AppendVector(A=Constant(), B=Multiply(A=@n16, B=@n17))
   Global Rotation
   Tiling=Constant2Vector() {g=1, r=1}
   Cell Offset=Constant()
   Cell Scale=Constant() {r=1.4}
   Cell Rotation=Constant()
   Random Seed
  ) {materialFunction=/FortUILibrary/MaterialFunctions/Basics/MF_UI_ParticleGrid}
  n19 = TextureObjectParameter() {
   group=03 - Streaks
   parameterName=Streaks Texture
   samplerType=SAMPLERTYPE_LinearColor
   texture=/Game/Catalog/BackgroundMaterialLibrary/Textures/T_UI_Streaks_Marvel
  }
  n20 = StaticComponentMaskParameter(None=TextureSample(UVs=@n18.Cells UV, Tex=@n19) {samplerSource=SSM_Wrap_WorldGroupSettings}.RGB) {defaultG=true, defaultR=true, group=03 - Streaks, parameterName=Streaks Texture Channels}
  n21 = Max(A=ComponentMask(None=@n20) {r=true}, B=ComponentMask(None=@n20) {g=true})
  n22 = ScalarParameter() {defaultValue=3, group=03 - Streaks, parameterName=Streaks Sharpening}
  NamedRerouteDeclaration(None=Saturate(None=Power(Base=@n21, Exp=@n22) {constExponent=3, desc=Sharpen Streaks})) {name=Streaks Mask, nodeColor=<rgba 0 0 0 1>}
  NamedRerouteDeclaration(None=@n11.Color 2) {name=Secondary Color, nodeColor=<rgba 0 0 0 1>}
  n23 = MaterialFunctionCall(UV=NamedRerouteDeclaration(None=MaterialFunctionCall() {materialFunction=&GetUserInterfaceUV}.Normalized UV) {name=Normalized UV, nodeColor=<rgba 0 0 0 1>}) {materialFunction=&MF_UI_UVScaling}
  NamedRerouteDeclaration(None=@n23.ScaleY) {name=Scale Y UV, nodeColor=<rgba 0 0 0 1>}
  n24 = NamedRerouteUsage()
  n25 = ScalarParameter() {group=01 - Gradient, parameterName=Gradient Rotation}
  n26 = NamedRerouteUsage()
  n27 = ScalarParameter() {defaultValue=1.2, group=01 - Gradient, parameterName=Radial Gradient Scale}
  n28 = ScalarParameter() {group=01 - Gradient, parameterName=Radial Gradient Offset X}
  n29 = ScalarParameter() {group=01 - Gradient, parameterName=Radial Gradient Offset Y}
  n30 = MaterialFunctionCall(UV=Add(
    A=MaterialFunctionCall(UVs=@n26, Texture Scale=@n27) {materialFunction=&ScaleUVsByCenter}.UVs
    B=AppendVector(A=@n28, B=@n29)
   ) {constB=0.5}) {materialFunction=&MF_UI_UVScaling}
  n31 = Saturate(None=StaticSwitchParameter(
    True=MaterialFunctionCall(UVs=@n24, RotationAngle=@n25) {materialFunction=&MF_UI_LinearGradient}.Result
    False=MaterialFunctionCall(UV=@n30.Scaled To Fill) {materialFunction=&MF_UI_RadialGradient}.Smooth
   ) {group=01 - Gradient, parameterName=Use Linear Gradient, sortPriority=-32})
  NamedRerouteDeclaration(None=LinearInterpolate(A=NamedRerouteUsage(), B=NamedRerouteUsage(), Alpha=NamedRerouteDeclaration(None=@n31) {name=Gradient Mask, nodeColor=<rgba 0 0 0 1>})) {name=Colored Gradient, nodeColor=<rgba 0 0 0 1>}
  NamedRerouteDeclaration(None=Multiply(A=Time(), B=ScalarParameter() {defaultValue=1, parameterName=Speed Factor})) {name=Speed, nodeColor=<rgba 0 0 0 1>}
  NamedRerouteDeclaration(None=@n23.Scaled To Fill) {name=Scale to Fill, nodeColor=<rgba 0 0 0 1>}
  NamedRerouteDeclaration(None=@n18.Random Columns) {name=Random Columns, nodeColor=<rgba 0 0 0 1>}
  NamedRerouteDeclaration(None=@n11.Color 3) {name=Accent Color, nodeColor=<rgba 0 0 0 1>}
 ]
}
