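# UI-domain translucent material for vehicle item shop backgrounds, combining a two-color gradient, a flag graphic and rounded-corner masking.
# Shared background-parameter colors blend through a gradient mask, with a static switch choosing rotatable linear or offsettable, scalable radial.
# An aspect-scaled flag alpha texture, sized by FlagScale, feeds a MultiplyAdd over a base color, faded in by FlagAlpha.
# Item shop material functions supply the colors and rounded corners.
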Material /Game/Catalog/BackgroundMaterialLibrary/CoreMaterials/M_UI_Shop_Background_VehicleItem
graph {
 properties {bAutomaticallySetUsageInEditor=false, blendMode=BLEND_Translucent, materialDomain=MD_UI, usedShadingModels=Default Lit}
 nodes [
  n1 = NamedRerouteUsage()
  n2 = Constant2Vector() {g=1}
  n3 = TextureObject() {samplerType=SAMPLERTYPE_Alpha, texture=/Game/Catalog/BackgroundMaterialLibrary/Textures/T_UI_Vehicle_Flag}
  n4 = MaterialFunctionCall(TextureSize=TextureProperty(Tex=@n3), ScalePivotPoint=Constant2Vector() {g=1}) {materialFunction=&MF_UI_UVScaling}
  n5 = Add(
   A=Multiply(A=Subtract(A=@n4.ScaleY, B=@n2) {constB=0.5}, B=ScalarParameter() {defaultValue=1, group=Flag, parameterName=FlagScale}) {constB=1.5}
   B=@n2
  ) {constB=0.5}
  LinearInterpolate(
   A=@n1
   B=MaterialFunctionCall!(
    Base=@n1
    Add=TextureSample(UVs=@n5, Tex=@n3) {samplerType=SAMPLERTYPE_Alpha, texture=&T_UI_Vehicle_Flag}.R
   ) {materialFunction=/Engine/Functions/Engine_MaterialFunctions02/Math/MultiplyAdd}.Result
   Alpha=ScalarParameter() {defaultValue=0.5, group=Flag, parameterName=FlagAlpha, sliderMax=1}
  ) {constAlpha=0.25}
  MaterialFunctionCall!(Input Alpha=Constant() {r=1}, bOverrideCornerRadius, Corner Radius Override) {materialFunction=/FortUILibrary/Systems/ItemShopTile/Materials/MaterialFunctions/MF_UI_ItemShop_RoundedCorners}
  n6 = MaterialFunctionCall() {materialFunction=/FortUILibrary/Systems/ItemShopTile/Materials/MaterialFunctions/MF_UI_ItemShop_BackgroundParams}
  NamedRerouteDeclaration(None=@n6.Color 1) {name=Primary Color, nodeColor=<rgba 0 0 0 1>}
  NamedRerouteDeclaration(None=MaterialFunctionCall() {materialFunction=&GetUserInterfaceUV}.Normalized UV) {name=Normalized UV, nodeColor=<rgba 0 0 0 1>}
  NamedRerouteDeclaration(None=@n6.Color 2) {name=Secondary Color, nodeColor=<rgba 0 0 0 1>}
  n7 = MaterialFunctionCall(UVs=NamedRerouteUsage(), RotationAngle=ScalarParameter() {group=01 - Gradient, parameterName=Gradient Rotation}) {materialFunction=&MF_UI_LinearGradient}
  n8 = ScalarParameter() {defaultValue=1, group=01 - Gradient, parameterName=Gradient Softness}
  n9 = NamedRerouteUsage()
  n10 = ScalarParameter() {defaultValue=1.2, group=01 - Gradient, parameterName=Radial Gradient Scale}
  n11 = ScalarParameter() {group=01 - Gradient, parameterName=Radial Gradient Offset X}
  n12 = ScalarParameter() {group=01 - Gradient, parameterName=Radial Gradient Offset Y}
  n13 = MaterialFunctionCall(UV=Add(
    A=MaterialFunctionCall(UVs=@n9, Texture Scale=@n10) {materialFunction=&ScaleUVsByCenter}.UVs
    B=AppendVector(A=@n11, B=@n12)
   ) {constB=0.5}) {materialFunction=&MF_UI_UVScaling}
  n14 = Saturate(None=StaticSwitchParameter(
    True=Power(Base=@n7.Result, Exp=@n8)
    False=MaterialFunctionCall(UV=@n13.Scaled To Fill) {materialFunction=&MF_UI_RadialGradient}.Smooth
   ) {group=01 - Gradient, parameterName=Use Linear Gradient, sortPriority=-32})
  NamedRerouteDeclaration(None=LinearInterpolate(A=NamedRerouteUsage(), B=NamedRerouteUsage(), Alpha=NamedRerouteDeclaration(None=@n14) {name=Gradient Mask, nodeColor=<rgba 0 0 0 1>})) {name=Colored Gradient, nodeColor=<rgba 0 0 0 1>}
 ]
}
