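# Translucent UI material for item shop vehicle backgrounds, showing a car in a stylized procedural environment.
# A radial backlight and a pseudo-3D rotated floor blend across a tilted horizon, then gain flow-map smoke, shop-colored hue-shifted tinting and a wave-distorted flag.
# A utility texture's channels drive car compositing, including reflection, bounce light and headlights, with an item stack image layered on top.
# Static switches toggle car, utility pass and environment, with swipe and rounded-corner alpha.
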Material /Game/Catalog/BackgroundMaterialLibrary/CoreMaterials/M_UI_Shop_Background_Vehicle_Web
graph {
 properties {
  bAutomaticallySetUsageInEditor=false
  blendMode=BLEND_Translucent
  materialDomain=MD_UI
  pixelDepthOffsetMode=PDOM_Legacy
  usedShadingModels=Default Lit
 }
 nodes [
  n1 = Constant2Vector() {g=1}
  n2 = TextureObject() {samplerType=SAMPLERTYPE_Alpha, texture=/Game/Catalog/BackgroundMaterialLibrary/Textures/T_UI_Vehicle_Flag}
  n3 = MaterialFunctionCall(TextureSize=TextureProperty(Tex=@n2), ScalePivotPoint=Constant2Vector() {g=1}) {materialFunction=&MF_UI_UVScaling}
  n4 = Add(
   A=Multiply(A=Subtract(A=@n3.ScaleY, B=@n1) {constB=0.5}, B=ScalarParameter() {defaultValue=2, group=Flag, parameterName=FlagScale}) {constB=1.5}
   B=@n1
  ) {constB=0.5}
  n5 = ComponentMask(None=MaterialFunctionCall!(
    UV
    Wave Height
    Wave Elevation
    Wave Length
    Wave Speed=Constant() {r=-0.25}
    GlowMin
    GlowMax
    Outline Thickness
    Outline Alignment
    Custom Curve
   ) {materialFunction=/FortUILibrary/MaterialFunctions/Shapes/MF_UI_Wave}.SDF) {r=true}
  Add(A=@n4, B=Multiply(A=Saturate(None=@n5)) {constB=0.025, desc=Waving Intensity}) {desc=Waving Flag}
  n6 = NamedRerouteUsage()
  n7 = Subtract(A=ComponentMask(None=@n6) {r=true}, B=ScalarParameter() {group=Car, parameterName=Car Offset X})
  n8 = Add(A=ComponentMask(None=@n6) {g=true}, B=ScalarParameter() {group=Car, parameterName=Car Offset Y})
  n9 = NamedRerouteDeclaration(None=MaterialFunctionCall(UVs=AppendVector(A=@n7, B=@n8), Texture Scale=ScalarParameter() {defaultValue=1, group=Car, parameterName=Car Scale}) {materialFunction=&ScaleUVsByCenter}.UVs) {name=Car UV, nodeColor=<rgba 0 0 0 1>}
  NamedRerouteDeclaration(None=TextureSample(UVs=@n9, Tex=NamedRerouteUsage()).RGBA) {name=Util Pass, nodeColor=<rgba 0 0 0 1>}
  n10 = TextureObjectParameter() {
   group=Car
   parameterName=CarUtil
   samplerType=SAMPLERTYPE_LinearColor
   texture=/Game/Catalog/BackgroundMaterialLibrary/Textures/T_UI_VehicleUtil
  }
  NamedRerouteDeclaration(None=@n10) {name=Car Util Texture, nodeColor=<rgba 0 0 0 1>}
  NamedRerouteDeclaration(None=MaterialFunctionCall() {materialFunction=/FortUILibrary/Systems/ItemShopTile/Materials/MaterialFunctions/MF_UI_ItemShop_BackgroundParams}.Color 1) {name=Accent Color, nodeColor=<rgba 0 0 0 1>}
  n11 = ScalarParameter() {defaultValue=0.01, group=Back Light, parameterName=Sky Min Opacity, sliderMax=1}
  n12 = ScalarParameter() {defaultValue=0.15, group=Back Light, parameterName=Back Light Opacity, sliderMax=1}
  n13 = NamedRerouteUsage()
  n14 = AppendVector(
   A=ScalarParameter() {defaultValue=0.5, group=Back Light, parameterName=Back Light Relative Offset X}
   B=ScalarParameter() {defaultValue=0.5, group=Back Light, parameterName=Back Light Relative Offset Y}
  )
  n15 = ScalarParameter() {defaultValue=1.5, group=Back Light, parameterName=Back Light Scale X}
  n16 = ScalarParameter() {defaultValue=3, group=Back Light, parameterName=Back Light Scale Y}
  n17 = MaterialFunctionCall(
   UV=MaterialFunctionCall(UVs=Add(A=@n13, B=@n14), Texture Scale=AppendVector(A=@n15, B=@n16)) {materialFunction=&ScaleUVsByCenter}.UVs
   TextureSize=NamedRerouteUsage() {desc=For stretching the backlight depending on ratio}
  ) {materialFunction=&MF_UI_UVScaling}
  n18 = AppendVector(
   A=ScalarParameter() {defaultValue=0.25, group=Back Light, parameterName=Back Light Absolute Offset X}
   B=ScalarParameter() {group=Back Light, parameterName=Back Light Absolute Offset Y}
  )
  n19 = MaterialFunctionCall("Value (0-1)"=MaterialFunctionCall(UV=Add(A=@n17.ScaleY, B=@n18)) {materialFunction=&MF_UI_RadialGradient}.Smooth) {materialFunction=&MF_UI_EaseCurves}
  n20 = ScalarParameter() {defaultValue=1, group=Back Light, parameterName=Backlight Intensity}
  n21 = Power(
   Base=LinearInterpolate(A=@n11, B=Max(A=@n11, B=@n12), Alpha=Multiply(A=@n19.EaseInCubic, B=@n20))
   Exp=ScalarParameter() {defaultValue=1, group=Mood, parameterName=Sky Contrast}
  ) {constExponent=1}
  n22 = VectorParameter() {group=Floor, parameterName=Floor Radial Angle, channelNames {a="", b=AngleZ, g=AngleY, r=AngleX}, defaultValue=<rgba -0.175 -0.05 0 1>}
  n23 = VectorParameter() {group=Floor, parameterName=Floor Radial Offset, channelNames {a="", b=OffsetZ, g=OffsetY, r=OffsetX}, defaultValue=<rgba 0 0 1.3021 1>}
  n24 = NamedRerouteDeclaration(None=MaterialFunctionCall!(
    UV=NamedRerouteUsage()
    Angle=@n22
    Camera FOV=Constant() {r=45}
    Camera Offset=@n23
    Offset from Pivot=Constant()
    Texture Scale
    Texture Size
   ) {materialFunction=/FortUILibrary/MaterialFunctions/Basics/MF_UI_Rotate3D}.Rotated UV) {name=Floor UV, nodeColor=<rgba 0 0 0 1>}
  n25 = AppendVector(
   A=ScalarParameter() {defaultValue=2, group=Floor, parameterName=Floor Radial Scale X}
   B=ScalarParameter() {defaultValue=1, group=Floor, parameterName=Floor Radial Scale Y}
  )
  n26 = NamedRerouteDeclaration(None=MaterialFunctionCall(UV=MaterialFunctionCall(UVs=@n24, Texture Scale=@n25) {materialFunction=&ScaleUVsByCenter}.UVs) {materialFunction=&MF_UI_RadialGradient}.Smooth) {name=Floor Radial Mask, nodeColor=<rgba 0 0 0 1>}
  n27 = LinearInterpolate(
   A=ScalarParameter() {defaultValue=0.0024, group=Mood, parameterName=Floor Outer Brightness, sliderMax=1}
   B=ScalarParameter() {defaultValue=1, group=Mood, parameterName=Floor Center Brightness, sliderMax=1}
   Alpha=NamedRerouteDeclaration(None=MaterialFunctionCall("Value (0-1)"=@n26) {materialFunction=&MF_UI_EaseCurves}.EaseInCubic) {name=Floor Radial Mask - Eased, nodeColor=<rgba 0 0 0 1>}
  ) {constA=0.2}
  n28 = LinearInterpolate(
   A=@n27
   B=LinearInterpolate(A=NamedRerouteUsage(), Alpha=@n27)
   Alpha=ScalarParameter() {group=Horizon Line, parameterName=Blend Floor and Sky, sliderMax=1}
  )
  n29 = Rotator(
   Coordinate=NamedRerouteDeclaration(None=NamedRerouteUsage()) {name=Scale Y, nodeColor=<rgba 0 0 0 1>}
   Time=ScalarParameter() {defaultValue=-0.045, group=Horizon Line, parameterName=Horizon Rotation}
  ) {speed=6.2832}
  n30 = AppendVector(A=Constant(), B=ScalarParameter() {defaultValue=0.0365, group=Horizon Line, parameterName=Horizon Offset Y})
  n31 = ComponentMask(None=NamedRerouteDeclaration(None=Add(A=@n29, B=@n30)) {name=Horizon UV, nodeColor=<rgba 0 0 0 1>}) {g=true}
  n32 = Multiply(A=ScalarParameter() {defaultValue=0.75, group=Horizon Line, parameterName=Horizon Sharpness, sliderMax=1}) {constB=0.5}
  n33 = MaterialFunctionCall("Value (0-1)"=MaterialFunctionCall(Input=Saturate(None=@n31), Input Low=@n32, Input High=OneMinus(None=@n32)) {materialFunction=&RemapValue_0-1}.Result) {desc=Horizon Mask, materialFunction=&MF_UI_EaseCurves}
  NamedRerouteDeclaration(None=LinearInterpolate(
    A=NamedRerouteDeclaration(None=@n21) {name=Backlight, nodeColor=<rgba 0 0 0 1>}
    B=NamedRerouteDeclaration(None=@n28) {name=Floor, nodeColor=<rgba 0 0 0 1>}
    Alpha=NamedRerouteDeclaration(None=@n33.EaseInOutQuint) {name=Environment Mask, nodeColor=<rgba 0 0 0 1>}
   )) {name=Environment Grayscale, nodeColor=<rgba 0 0 0 1>}
  n34 = NamedRerouteUsage()
  n35 = Saturate(None=NamedRerouteUsage())
  n36 = OneMinus(None=@n35)
  n37 = ComponentMask(None=NamedRerouteDeclaration(None=Min(A=@n35, B=@n36)) {name=Horizon Gradients, nodeColor=<rgba 0 0 0 1>}) {g=true}
  n38 = Constant() {r=0.3}
  n39 = OneMinus(None=@n38)
  n40 = Multiply(A=MaterialFunctionCall("Value (0-1)"=MaterialFunctionCall(Input=@n37, Input Low=@n38, Input High=@n39) {desc=Smoke Area Mask, materialFunction=&RemapValue_0-1}.Result) {materialFunction=&MF_UI_EaseCurves}.EaseInCubic) {constB=8}
  n41 = ComponentMask(None=MaterialFunctionCall!(
    UVs=NamedRerouteUsage()
    Texture Object=TextureObject() {samplerType=SAMPLERTYPE_Alpha, texture=/Game/Catalog/BackgroundMaterialLibrary/Textures/T_UI_Vehicle_Smoke}
    Grayscale Flow Map=Constant() {r=0.5}
    Flow Direction=ScalarParameter() {defaultValue=-0.8, desc=Smoke Orientation, group=Horizon Line, parameterName=Smoke Orientation}
    Flow Intensity
    Flow Speed
   ) {materialFunction=/FortUILibrary/MaterialFunctions/Effects/MF_UI_GrayscaleFlowMap_New}.Distorted Texture) {r=true}
  n42 = Multiply(A=Multiply(A=NamedRerouteUsage(), B=Saturate(None=@n40)), B=SquareRoot(None=Multiply(A=@n41, B=@n41)))
  n43 = LinearInterpolate(
   A=ScalarParameter() {defaultValue=1, group=Mood, parameterName=Smoke Left Intensity}
   B=ScalarParameter() {defaultValue=1, group=Mood, parameterName=Smoke Right Intensity}
   Alpha=ComponentMask(None=NamedRerouteUsage()) {r=true}
  )
  n44 = NamedRerouteDeclaration(None=Saturate(None=Multiply(A=@n42, B=@n43))) {name=Smoke, nodeColor=<rgba 0 0 0 1>}
  n45 = Multiply(A=ComponentMask(None=NamedRerouteUsage()) {r=true}) {constB=2}
  n46 = LinearInterpolate(A=Multiply(A=@n44) {constB=0.1}, B=@n44, Alpha=LinearInterpolate(A=Saturate(None=@n45), Alpha=NamedRerouteUsage()))
  n47 = Add(A=@n46, B=NamedRerouteUsage())
  n48 = LinearInterpolate(A=LinearInterpolate(A=@n46, B=Saturate(None=@n47), Alpha=NamedRerouteUsage()), Alpha=NamedRerouteUsage()) {constB=0}
  NamedRerouteDeclaration(None=Saturate(None=Add(A=@n34, B=@n48))) {name=Grayscale Environment and Smoke, nodeColor=<rgba 0 0 0 1>}
  n49 = ScalarParameter() {defaultValue=0.04, group=Mood, parameterName=Hue Shift, sliderMax=1}
  n50 = MaterialFunctionCall() {materialFunction=&MF_UI_ItemShop_BackgroundParams}
  n51 = ComponentMask(None=MaterialFunctionCall() {materialFunction=&GetUserInterfaceUV}.Normalized UV) {g=true}
  n52 = Divide(
   A=MaterialFunctionCall(
    UVs=@n51
    GradientStart=ScalarParameter() {group=Color, parameterName=Gradient Start, sortPriority=0}
    GradientEnd=ScalarParameter() {defaultValue=1, group=Color, parameterName=Gradient End, sortPriority=2}
    GradientOffset=ScalarParameter() {defaultValue=-0.05, group=Color, parameterName=Gradient Offset, sortPriority=3}
   ) {materialFunction=&MF_UI_LinearGradient}.Result
   B=ScalarParameter() {defaultValue=0.95, group=Color, parameterName=Gradient Scale, sortPriority=4}
  )
  n53 = NamedRerouteUsage()
  n54 = Multiply(A=@n53) {constB=0.5}
  n55 = NamedRerouteDeclaration(None=StaticSwitchParameter(True=LinearInterpolate(A=@n50.Color 1, B=@n50.Color 1, Alpha=@n52), False=LinearInterpolate(A=@n53, B=@n54, Alpha=@n51)) {group=Color, parameterName=Use Background Color 2, sortPriority=33}) {name=Background Color, nodeColor=<rgba 0 0 0 1>}
  n56 = NamedRerouteDeclaration(None=LinearInterpolate(
    A=NamedRerouteUsage()
    B=MaterialFunctionCall(Hue Shift Percentage=@n49, Texture=@n55) {materialFunction=&HueShift}.Result
    Alpha=NamedRerouteUsage()
   )) {name=Colored Environment, nodeColor=<rgba 0 0 0 1>}
  n57 = NamedRerouteUsage()
  n58 = LinearInterpolate(A=Multiply(A=@n56, B=@n57), B=@n56, Alpha=@n57)
  n59 = TextureSample(UVs=@n4, Tex=@n2) {samplerType=SAMPLERTYPE_Alpha, texture=&T_UI_Vehicle_Flag}
  NamedRerouteDeclaration(None=LinearInterpolate(
    A=@n58
    B=MaterialFunctionCall!(Base=@n58, Add=@n59.R) {materialFunction=/Engine/Functions/Engine_MaterialFunctions02/Math/MultiplyAdd}.Result
    Alpha=ScalarParameter() {defaultValue=0.5, group=Flag, parameterName=FlagAlpha, sliderMax=1}
   ) {constAlpha=0.25}) {name=Environment Pass, nodeColor=<rgba 0 0 0 1>}
  NamedRerouteDeclaration(None=TextureObjectParameter() {group=Car, parameterName=CarTexture, texture=/Game/Catalog/BackgroundMaterialLibrary/Textures/T_UI_VehicleBeauty}) {name=Car Texture, nodeColor=<rgba 0 0 0 1>}
  NamedRerouteDeclaration(None=MaterialFunctionCall() {materialFunction=&GetUserInterfaceUV}.Pixel Size) {name=Pixel Dimension, nodeColor=<rgba 0 0 0 1>}
  n60 = MaterialFunctionCall!(Texture Size=TextureProperty(Tex=NamedRerouteUsage()), UVs) {materialFunction=/FortUILibrary/Systems/ItemShopTile/Materials/MaterialFunctions/MF_UI_ItemShop_OfferImage_RulesetUVs}
  n61 = Add(A=Multiply(A=Subtract(A=@n60.UVs) {constB=0.5}, B=ScalarParameter() {defaultValue=1, group=Utils, parameterName=Global Scale})) {constB=0.5}
  n62 = Subtract(A=ComponentMask(None=@n61) {r=true}, B=ScalarParameter() {group=Utils, parameterName=Global Offset X})
  n63 = Add(A=ComponentMask(None=@n61) {g=true}, B=ScalarParameter() {group=Utils, parameterName=Global Offset Y})
  NamedRerouteDeclaration(None=Rotator(Coordinate=AppendVector(A=@n62, B=@n63), Time=ScalarParameter() {group=Utils, parameterName=Global Rotation}) {speed=6.2832}) {name=Global UV, nodeColor=<rgba 0 0 0 1>}
  ScalarParameter() {group=Item Stack, parameterName=Item Stack Offset X}
  ScalarParameter() {defaultValue=0.5, group=Item Stack, parameterName=Item Stack Offset Y}
  n64 = NamedRerouteUsage()
  n65 = TextureSample(UVs=@n9, Tex=NamedRerouteUsage())
  n66 = NamedRerouteDeclaration(None=@n65.A) {name=Car Alpha, nodeColor=<rgba 0 0 0 1>}
  n67 = ScalarParameter() {defaultValue=0.15, group=Floor, parameterName=Reflection Amount, sliderMax=1}
  n68 = NamedRerouteDeclaration(None=LinearInterpolate(
    A=NamedRerouteUsage()
    B=NamedRerouteDeclaration(None=@n65.RGB) {name=Car RGB, nodeColor=<rgba 0 0 0 1>}
    Alpha=Multiply(A=@n66, B=@n67) {constB=0.15}
   )) {name=Reflection Pass, nodeColor=<rgba 0 0 0 1>}
  n69 = Multiply(A=@n68) {constB=0.01}
  n70 = ComponentMask(None=NamedRerouteUsage()) {b=true}
  n71 = ComponentMask(None=NamedRerouteUsage()) {r=true}
  n72 = NamedRerouteUsage()
  n73 = NamedRerouteUsage()
  n74 = MaterialFunctionCall!(Blend=NamedRerouteUsage(), Base=@n73) {materialFunction=/Engine/Functions/Engine_MaterialFunctions03/Blends/Blend_Screen}
  n75 = Multiply(A=ComponentMask(None=NamedRerouteUsage()) {g=true}, B=ScalarParameter() {defaultValue=1, parameterName=Bounce Light})
  n76 = MaterialFunctionCall(
   Blend=Multiply(A=ComponentMask(None=@n64) {a=true}, B=VectorParameter() {group=Utils, parameterName=Headlight Color, defaultValue=<rgba 1 1 1 1>}.RGB)
   Base=LinearInterpolate(
    A=LinearInterpolate(A=@n68, B=@n69, Alpha=@n70) {constB=0}
    B=LinearInterpolate(A=@n73, B=@n74.Result, Alpha=@n75)
    Alpha=Multiply(A=@n71, B=@n72)
   )
  ) {materialFunction=&Blend_Screen}
  n77 = NamedRerouteUsage()
  n78 = LinearInterpolate(A=NamedRerouteUsage(), B=Multiply(A=NamedRerouteUsage(), B=@n77), Alpha=@n77)
  n79 = TextureObjectParameter() {group=Item Stack, parameterName=ItemStackTexture, texture=/FortUILibrary/Input/PeripheralGuitar/Source/Empty1x1}
  n80 = MaterialFunctionCall(
   UV=MaterialFunctionCall() {materialFunction=&GetUserInterfaceUV}.Normalized UV
   TextureSize=TextureProperty(Tex=@n79)
   ScalePivotPoint=AppendVector(
    A=ScalarParameter() {group=Item Stack, parameterName=Item Stack Scale Pivot X}
    B=ScalarParameter() {defaultValue=0.5, group=Item Stack, parameterName=Item Stack Scale Pivot Y}
   )
  ) {materialFunction=&MF_UI_UVScaling}
  n81 = ComponentMask(None=@n80.ScaleY) {r=true}
  n82 = Constant() {r=-0.089}
  n83 = ComponentMask(None=@n80.ScaleY) {g=true}
  n84 = Constant() {r=-0.004}
  n85 = MaterialFunctionCall(
   UVs=AppendVector(A=Subtract(A=@n81, B=@n82), B=Add(A=@n83, B=@n84))
   Texture Scale=ScalarParameter() {defaultValue=1, group=Item Stack, parameterName=Item Stack Scale}
  ) {materialFunction=&ScaleUVsByCenter}
  n86 = TextureSample(UVs=@n85.UVs, Tex=@n79) {samplerSource=SSM_Clamp_WorldGroupSettings}
  StaticSwitchParameter(
   True=LinearInterpolate(
    A=StaticSwitchParameter(True=@n76.Result, False=@n78) {defaultValue=true, group=Car, parameterName=Use Utility Pass, sortPriority=-32}
    B=@n86.RGB
    Alpha=MaterialFunctionCall!(UVs=@n85.UVs, Alpha=@n86.A) {materialFunction=/FortUILibrary/MaterialFunctions/Basics/MF_UI_ClipEdgePixel}.Result
   )
   False=NamedRerouteUsage()
  ) {defaultValue=true, group=Utils, parameterName=Display Car}
  n87 = NamedRerouteUsage()
  n88 = ComponentMask(None=NamedRerouteUsage()) {b=true}
  n89 = NamedRerouteUsage()
  StaticSwitchParameter(
   True=MaterialFunctionCall!(
    Input Alpha=MaterialFunctionCall() {materialFunction=/FortUILibrary/Systems/ItemShopTile/Materials/MaterialFunctions/MF_UI_ItemShop_OfferImage_Swiping}.Swipe alpha mask
    bOverrideCornerRadius
    Corner Radius Override
   ) {materialFunction=/FortUILibrary/Systems/ItemShopTile/Materials/MaterialFunctions/MF_UI_ItemShop_RoundedCorners}.Final Alpha
   False=LinearInterpolate(A=LinearInterpolate(A=@n87, Alpha=@n88), Alpha=ComponentMask(None=@n89) {a=true})
  ) {defaultValue=true, group=Utils, parameterName=Display Environment}
 ]
}
